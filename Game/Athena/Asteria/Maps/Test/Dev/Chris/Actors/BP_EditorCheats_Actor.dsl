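; Dev test actor for Fortnite that automatically runs a list of cheat console commands when play begins.
; On BeginPlay, the authority waits a short delay, then gathers all Fort player controllers.
; For each controller it executes every entry in the GetAllCheatCvarsToCall array, without passing that controller, so the list repeats per player.
; A player-count branch precedes the loop, but neither side holds logic of its own.
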(event EventBeginPlay
  (Networking|SwitchHasAuthority
    (:Authority
      (Utilities|FlowControl|Delay 1.0)
      (if (> (Utilities|Array|Length (Fortnite|GetAllFortPlayerControllers true)) 0)
        _
        (else
          _))
      (for _ (Fortnite|GetAllFortPlayerControllers true)
        (for _array_element (Variables|Default|GetAllCheatCvarsToCall)
          (Development|ExecuteConsoleCommand _array_element))))))

(fn ConstructionScript)
