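; Test blade actor that knocks overlapping pawns and vehicles away, optionally applying environment damage.
; The nearest of four cached arrow components supplies forward and negated-right vectors, scaled by VelocityForce with added lift, for LaunchCharacter or ImpulseVehicle.
; Begin-overlap applies this immediately if no timer is active, and starts a looping timer that repeats it until nothing overlaps.
; A note says vehicle impulses need both client and server.
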(event EventBeginPlay
  ;[1 1 1] Init
  (CallFunction|CacheArrows)
  (Networking|SwitchHasAuthority
    (:Authority
      _)
    (:Remote
      _)) ; vehicle impulse needs to be client + server for some reason
  (Collision|BindEventtoOnActorBeginOverlap (delegate Custom|OnActorBeginOverlap_Event)))

(event Custom|OnActorBeginOverlap_Event
  (params (OverlappedActor :type Actor%20Object%20Reference) (OtherActor :type Actor%20Object%20Reference))
  (bind _collisionmeshouter (Variables|Default|GetCollisionMeshOuter))
  (bind _overlappingactors (Collision|GetOverlappingActors _collisionmeshouter "/Script/FortniteGame.FortPawn" :!class /Script/Engine.PrimitiveComponent))
  (bind _collisionmeshouter_2 (Variables|Default|GetCollisionMeshOuter))
  (bind _overlappingactors_2 (Collision|GetOverlappingActors _collisionmeshouter_2 "/Script/FortniteGame.FortAthenaVehicle" :!class /Script/Engine.PrimitiveComponent))
  ;[1 1 1] Init
  (Variables|Default|SetCurrentActorRef OtherActor)
  (if (not (Utilities|Time|IsTimerActivebyHandle (Variables|Default|GetTimerHandle)))
    (bind _returnvalue (Utilities|Time|SetTimerbyEvent (delegate Custom|CheckOverlaps) 0.75 true))
    (Variables|Default|SetTimerHandle _returnvalue)
    ;[1 1 1] Vehicles
    (if (> (Utilities|Array|Length _overlappingactors_2) 0)
      (Variables|Default|SetNoVehicles false)
      (for _array_element _overlappingactors_2
        (bind _output_get (Variables|Default|SetCurrentActorRef _array_element))
        (Utilities|Casting|CastToFortAthenaVehicle _output_get)
        (bind _arrowindex (CallFunction|GetNearestBladeDirectionArrowIndex (Transformation|GetActorLocation (Variables|Default|GetCurrentActorRef))))
        (bind (_forwardvector _rightvector) (CallFunction|CalculateLeftRightVectorImpulse _arrowindex))
        (Variables|Default|SetCurrentForwardVector _forwardvector)
        (Variables|Default|SetCurrentRightVector _rightvector)
        (bind _asfort_athena_vehicle (Utilities|Casting|CastToFortAthenaVehicle (Variables|Default|GetCurrentActorRef)))
        (Vehicle|ImpulseVehicle _asfort_athena_vehicle (+ (+ (* (Variables|Default|GetCurrentForwardVector) (* (Variables|Default|GetVelocityForce) "1.500000,1.500000,1.500000")) "0.000000,0.000000,325.000000") (* (Variables|Default|GetCurrentRightVector) (Variables|Default|GetVelocityForce))) (Math|Vector|MakeVectorNetQuantize 0.0) false false false "None")
        (if (Variables|Default|GetDamage)
          (Environment|ApplyGlobalEnvironmentDamagetoActor _asfort_athena_vehicle 2e+02)))
      (else
        (Variables|Default|SetNoVehicles true)))
    ;[1 1 1] Pawns
    (if (> (Utilities|Array|Length _overlappingactors) 0)
      (Variables|Default|SetNoPlayers false)
      (for _array_element_2 _overlappingactors
        (Variables|Default|SetCurrentActorRef _array_element_2)
        (bind _arrowindex_2 (CallFunction|GetNearestBladeDirectionArrowIndex (Transformation|GetActorLocation (Variables|Default|GetCurrentActorRef))))
        (bind (_forwardvector_2 _rightvector_2) (CallFunction|CalculateLeftRightVectorImpulse _arrowindex_2))
        (Variables|Default|SetCurrentForwardVector _forwardvector_2)
        (Variables|Default|SetCurrentRightVector _rightvector_2)
        (bind _asfort_pawn (Utilities|Casting|CastToFortPawn (Variables|Default|GetCurrentActorRef)))
        (Variables|Default|SetFinalPlayerlaunchVector (+ (+ (* (Variables|Default|GetCurrentForwardVector) (Variables|Default|GetVelocityForce)) "0.000000,0.000000,700.000000") (* (Variables|Default|GetCurrentRightVector) (Variables|Default|GetVelocityForce))))
        (Character|LaunchCharacter _asfort_pawn (Variables|Default|GetFinalPlayerlaunchVector) true true)
        (if (Variables|Default|GetDamage)
          (Environment|ApplyGlobalEnvironmentDamagetoActor (Variables|Default|GetCurrentActorRef) 9999.0)))
      (else
        (Variables|Default|SetNoPlayers true)))
    ;[1 1 1] Stop check if nothing overlapping
    (if (and (Variables|Default|GetNoVehicles) (Variables|Default|GetNoPlayers))
      (Utilities|Time|PauseTimerbyHandle (Variables|Default|GetTimerHandle)))))

(event Custom|CheckOverlaps
  (bind _collisionmeshouter (Variables|Default|GetCollisionMeshOuter))
  (bind _overlappingactors (Collision|GetOverlappingActors _collisionmeshouter "/Script/FortniteGame.FortPawn" :!class /Script/Engine.PrimitiveComponent))
  (bind _collisionmeshouter_2 (Variables|Default|GetCollisionMeshOuter))
  (bind _overlappingactors_2 (Collision|GetOverlappingActors _collisionmeshouter_2 "/Script/FortniteGame.FortAthenaVehicle" :!class /Script/Engine.PrimitiveComponent))
  ;[1 1 1] Vehicles
  (if (> (Utilities|Array|Length _overlappingactors_2) 0)
    (Variables|Default|SetNoVehicles false)
    (for _array_element _overlappingactors_2
      (bind _output_get (Variables|Default|SetCurrentActorRef _array_element))
      (Utilities|Casting|CastToFortAthenaVehicle _output_get)
      (bind _arrowindex (CallFunction|GetNearestBladeDirectionArrowIndex (Transformation|GetActorLocation (Variables|Default|GetCurrentActorRef))))
      (bind (_forwardvector _rightvector) (CallFunction|CalculateLeftRightVectorImpulse _arrowindex))
      (Variables|Default|SetCurrentForwardVector _forwardvector)
      (Variables|Default|SetCurrentRightVector _rightvector)
      (bind _asfort_athena_vehicle (Utilities|Casting|CastToFortAthenaVehicle (Variables|Default|GetCurrentActorRef)))
      (Vehicle|ImpulseVehicle _asfort_athena_vehicle (+ (+ (* (Variables|Default|GetCurrentForwardVector) (* (Variables|Default|GetVelocityForce) "1.500000,1.500000,1.500000")) "0.000000,0.000000,325.000000") (* (Variables|Default|GetCurrentRightVector) (Variables|Default|GetVelocityForce))) (Math|Vector|MakeVectorNetQuantize 0.0) false false false "None")
      (if (Variables|Default|GetDamage)
        (Environment|ApplyGlobalEnvironmentDamagetoActor _asfort_athena_vehicle 2e+02)))
    (else
      (Variables|Default|SetNoVehicles true)))
  ;[1 1 1] Pawns
  (if (> (Utilities|Array|Length _overlappingactors) 0)
    (Variables|Default|SetNoPlayers false)
    (for _array_element_2 _overlappingactors
      (Variables|Default|SetCurrentActorRef _array_element_2)
      (bind _arrowindex_2 (CallFunction|GetNearestBladeDirectionArrowIndex (Transformation|GetActorLocation (Variables|Default|GetCurrentActorRef))))
      (bind (_forwardvector_2 _rightvector_2) (CallFunction|CalculateLeftRightVectorImpulse _arrowindex_2))
      (Variables|Default|SetCurrentForwardVector _forwardvector_2)
      (Variables|Default|SetCurrentRightVector _rightvector_2)
      (bind _asfort_pawn (Utilities|Casting|CastToFortPawn (Variables|Default|GetCurrentActorRef)))
      (Variables|Default|SetFinalPlayerlaunchVector (+ (+ (* (Variables|Default|GetCurrentForwardVector) (Variables|Default|GetVelocityForce)) "0.000000,0.000000,700.000000") (* (Variables|Default|GetCurrentRightVector) (Variables|Default|GetVelocityForce))))
      (Character|LaunchCharacter _asfort_pawn (Variables|Default|GetFinalPlayerlaunchVector) true true)
      (if (Variables|Default|GetDamage)
        (Environment|ApplyGlobalEnvironmentDamagetoActor (Variables|Default|GetCurrentActorRef) 9999.0)))
    (else
      (Variables|Default|SetNoPlayers true)))
  ;[1 1 1] Stop check if nothing overlapping
  (if (and (Variables|Default|GetNoVehicles) (Variables|Default|GetNoPlayers))
    (Utilities|Time|PauseTimerbyHandle (Variables|Default|GetTimerHandle))))

(fn CacheArrows
  (Utilities|Array|Add (Variables|Default|GetAllDirectionArrowComponents) (Variables|Default|GetArrow))
  (Utilities|Array|Add (Variables|Default|GetAllDirectionArrowComponents) (Variables|Default|GetArrow1))
  (Utilities|Array|Add (Variables|Default|GetAllDirectionArrowComponents) (Variables|Default|GetArrow2))
  (Utilities|Array|Add (Variables|Default|GetAllDirectionArrowComponents) (Variables|Default|GetArrow3)))

(fn CalculateLeftRightVectorImpulse
  (params (Index :type Integer))
  (bind _arrow1 (Variables|Default|GetArrow1))
  (bind _arrow2 (Variables|Default|GetArrow2))
  (bind _arrow3 (Variables|Default|GetArrow3))
  (bind _arrow (Variables|Default|GetArrow))
  (switch int Index
    (:0
      (LocalVariables|SetForwardVectorLocal (Transformation|GetForwardVector _arrow))
      (LocalVariables|SetRightVectorLocal (Math|Vector|NegateVector (Transformation|GetRightVector _arrow))))
    (:1
      (LocalVariables|SetForwardVectorLocal (Transformation|GetForwardVector _arrow1))
      (LocalVariables|SetRightVectorLocal (Math|Vector|NegateVector (Transformation|GetRightVector _arrow1))))
    (:2
      (LocalVariables|SetForwardVectorLocal (Transformation|GetForwardVector _arrow2))
      (LocalVariables|SetRightVectorLocal (Math|Vector|NegateVector (Transformation|GetRightVector _arrow2))))
    (:3
      (LocalVariables|SetForwardVectorLocal (Transformation|GetForwardVector _arrow3))
      (LocalVariables|SetRightVectorLocal (Math|Vector|NegateVector (Transformation|GetRightVector _arrow3)))))
  (return (LocalVariables|GetForwardVectorLocal) (LocalVariables|GetRightVectorLocal)))

(fn GetNearestBladeDirectionArrowIndex
  (params (ImpactVector :type Vector))
  (LocalVariables|SetVectorLocal ImpactVector)
  (for _array_element (Variables|Default|GetAllDirectionArrowComponents)
    (Utilities|Array|Add (LocalVariables|GetAllDistancesLocal) (Math|Vector|Distance(Vector) (LocalVariables|GetVectorLocal) (Transformation|GetWorldLocation _array_element))))
  (return (Math|Float|MinOfFloatArray (LocalVariables|GetAllDistancesLocal))))

(fn OnRep_CurrentRightVector)

(fn ConstructionScript)
